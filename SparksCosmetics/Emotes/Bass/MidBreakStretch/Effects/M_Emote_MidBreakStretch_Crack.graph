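# Translucent unlit Niagara sprite material for the MidBreakStretch emote's crack effect, drawing a stylized layered graphic from three mask textures.
# Hard-stepped, offset and scaled masks layer outline, midline and innerline colors, with gradient-colored halftone dots whose size varies along a rotatable axis.
# Particle color tints the emissive result.
# The combined mask is eroded by Voronoi noise driven by the Erode dynamic parameter, then stepped to a hard edge.
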Material /SparksCosmetics/Emotes/Bass/MidBreakStretch/Effects/M_Emote_MidBreakStretch_Crack
graph {
 properties {
  bAutomaticallySetUsageInEditor=false
  bUsedWithNiagaraSprites=true
  blendMode=BLEND_Translucent
  shadingModel=MSM_Unlit
  usedShadingModels=Unlit
 }
 nodes [
  Constant2Vector() {g=0.9, r=0.9}
  Constant2Vector() {g=0.5, r=0.5}
  Constant2Vector() {g=0.025, r=0.025}
  n1 = VectorParameter() {
   parameterName=Line Offset
   channelNames {a=Innerline Y, b=Innerline X, g=Midline Y, r=Midline X}
   defaultValue=<rgba 0.025 0.025 0.025 0.025>
  }
  n2 = VectorParameter() {
   parameterName=Texture Scale
   channelNames {a=Texture 2 - V Scale, b=Texture 2 - U Scale, g=Texture 1 - V Scale, r=Texture 1 - U Scale}
   defaultValue=<rgba 0.9 0.9 0.5 0.5>
  }
  n3 = MaterialFunctionCall(
   UVs=Add(A=TextureCoordinate(), B=AppendVector(A=@n1.Midline X, B=@n1.Midline X))
   Texture Scale=AppendVector(A=@n2.Texture 1 - U Scale, B=@n2.Texture 1 - U Scale)
  ) {materialFunction=&ScaleUVsByCenter}
  n4 = Multiply(
   A=@n3.0-1 mask
   B=TextureSampleParameter2D(UVs=@n3.UVs) {parameterName=Texture 1, texture=/Game/Effects/Fort_Effects/Textures/Flares/T_Emoij_Fallback}.G
  )
  n5 = Step(X=@n4) {constY=0.5}
  n6 = TextureCoordinate()
  n7 = ScalarParameter() {defaultValue=20, parameterName=DotCount}
  n8 = Frac(None=Multiply(A=TextureCoordinate(), B=AppendVector(A=@n7, B=@n7)))
  n9 = ComponentMask(None=MaterialFunctionCall(UVs=TextureCoordinate(), "Rotation Angle (0-1)"=ScalarParameter() {parameterName=Dots Rotation Angle}) {materialFunction=&CustomRotator}.Rotated Values) {r=true}
  n10 = LinearInterpolate(Alpha=SmoothStep(Value=@n9)) {constA=0.1, constB=0.5}
  n11 = LinearInterpolate(Alpha=SmoothStep(Value=@n9)) {constB=50}
  n12 = LinearInterpolate(
   A=LinearInterpolate(
    A=VectorParameter() {parameterName=Outline Color, defaultValue=<rgba 0 0 0 1>}.RGB
    B=VectorParameter() {parameterName=Midline Color, defaultValue=<rgba 1 0.0643 0 1>}.RGB
    Alpha=@n5
   )
   B=LinearInterpolate(
    A=VectorParameter() {parameterName=Gradient Color 1, defaultValue=<rgba 0 0.8876 1 1>}.RGB
    B=VectorParameter() {parameterName=Gradient Color 2, defaultValue=<rgba 0.1064 1 0 1>}.RGB
    Alpha=ComponentMask(None=@n6) {r=true}
   )
   Alpha=Multiply(
    A=MaterialFunctionCall(UVs=@n8, Radius=@n10, Density=@n11) {materialFunction=&RadialGradientExponential}.RadialGradientExponential
    B=@n4
   )
  )
  n13 = VectorParameter() {parameterName=Innerline Color, defaultValue=<rgba 0 0 1 1>}
  n14 = MaterialFunctionCall(
   UVs=Add(A=TextureCoordinate(), B=AppendVector(A=@n1.Innerline X, B=@n1.Innerline X))
   Texture Scale=AppendVector(A=@n2.Texture 2 - U Scale, B=@n2.Texture 2 - U Scale)
  ) {materialFunction=&ScaleUVsByCenter}
  n15 = Step(X=Multiply(
    A=@n14.0-1 mask
    B=TextureSampleParameter2D(UVs=@n14.UVs) {parameterName=Texture 2, texture=&T_Emoij_Fallback}.G
   )) {constY=0.5}
  MaterialFunctionCall(Emissive=Multiply(A=LinearInterpolate(A=@n12, B=@n13.RGB, Alpha=@n15), B=ParticleColor().RGB)) {materialFunction=&MF_FX_EmissiveAdjustments}
  n16 = Add(
   A=@n5
   B=Step(X=TextureSampleParameter2D() {parameterName=Texture 3, texture=&T_Emoij_Fallback}.G) {constY=0.5}
  )
  n17 = DynamicParameter() {defaultValue=<rgba 0 0 1 1>, paramNames [Erode,Param2,Param3,Param4]}
  n18 = Add(A=@n17.Erode, B=ScalarParameter() {defaultValue=-0.1, parameterName=Reveal Edge Length}) {constB=-0.1}
  n19 = ScalarParameter() {defaultValue=1, parameterName=Erode Scale U}
  n20 = ScalarParameter() {defaultValue=1, parameterName=Erode Scale V}
  n21 = TextureSample(UVs=Multiply(A=TextureCoordinate(), B=AppendVector(A=@n19, B=@n20))) {samplerType=SAMPLERTYPE_LinearColor, texture=/Game/Global/Materials/MaterialGym/Materials/TechArtPatterns/Textures/T_Voronoi_DF}
  Step(X=Multiply(A=Saturate(None=@n16), B=SmoothStep(Min=@n18, Max=@n17.Erode, Value=@n21.R) {constMax=0, constMin=1})) {constY=0.5}
  TextureSample() {samplerType=SAMPLERTYPE_Normal, texture=/Game/Global/Textures/Default/Masks_and_Gradients/T_MF_Crystal3}
 ]
}
