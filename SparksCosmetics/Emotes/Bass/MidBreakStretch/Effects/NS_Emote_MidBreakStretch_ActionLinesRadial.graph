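# Niagara system for radial action lines in the Sparks bass MidBreakStretch emote.
# One sprite emitter spawns thin, short-lived black streaks around a sphere, giving them velocity damped by random drag.
# Over life each streak thins and lengthens, with a negative camera offset.
# User parameters multiply spawn rate and lifetime, both defaulting to zero; a front-end size modifier is wired but neutral.
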NiagaraSystem /SparksCosmetics/Emotes/Bass/MidBreakStretch/Effects/NS_Emote_MidBreakStretch_ActionLinesRadial
dependencies {usedDataInterfaces [&NiagaraDataInterfaceCurve], usedDynamicInputs [], usedModules [&SystemState]}
emitters {ActionLines {
  emitterUpdateScript [
   &EmitterState()
   &SpawnRate(SpawnRate=&Multiply_Float(A=70, B=$User.ActionLinesSpawnRate), Use Spawn Probability=false)
  ]
  particleSpawnScript [
   &InitializeParticle(
    Color=<rgba 0 0 0 1>
    Lifetime=&Multiply_Float(A=&RandomRangeFloat(Minimum=0.15, Maximum=0.3), B=$User.ActionLinesLifetime)
    Sprite Size=&Multiply_Vector2D_ByFloat(
     Float=&FN_FrontEndFloatModifier(Gameplay=1, Front End=1)
     Vector2D=&RandomRangeVector2D(Minimum=<1 40>, Maximum=<1 60>)
    )
    Sprite Size Mode=::NewEnumerator4
    UsePositionOffset=false
    Write Material Random=true
   )
   &ShapeLocation(
    MaskSpawnGroup=false
    Offset Coordinate Space=::Local
    Offset Mode=::NewEnumerator0
    Override Seed=false
    Sphere Radius=30
    Sphere Surface Distribution=1
    UseOwnerScale=true
   )
   &AddVelocity(Module.Use Falloff Distance=false, Override Seed=false, Velocity Mode=::NewEnumerator1, Velocity Speed=1) {module=AddVelocity001}
  ]
  particleUpdateScript [
   &ParticleState()
   &Drag(Drag=&RandomRangeFloat(Minimum=18, Maximum=24), Use Linear Drag=true, Use Rotational Drag=false)
   &SolveForcesAndVelocity(Clamp Velocity=false, Limit Acceleration=false)
   &CameraOffset(Camera Offset Amount=-60)
   &ScaleSpriteSize(
    Scale Factor=&MakeVector2D(
     X=&FloatFromCurve(FloatCurve=DataInterface() {propertyValues {Curve {keys [<at 0 1 RCIM_Cubic>, {arriveTangent=-3.15, interpMode=RCIM_Cubic, leaveTangent=-3.15, tangentMode=RCTM_User, time=1, value=0}]}}})
     Y=&FloatFromCurve(FloatCurve=DataInterface() {propertyValues {Curve {keys [{arriveTangent=2.8355, interpMode=RCIM_Cubic, leaveTangent=2.8355, tangentMode=RCTM_User, time=0, value=0}, <at 1 1 RCIM_Cubic>]}}})
    )
    Scale Sprite Size Mode=::NewEnumerator2
   ) {module=ScaleSpriteSize001}
  ]
  rendererClasses [&NiagaraSpriteRendererProperties]
  renderers [{rendererClass=&NiagaraSpriteRendererProperties}]
 }}
summary {
 systemName=NS_Emote_MidBreakStretch_ActionLinesRadial
 emitters [{emitterName=ActionLines, rendererClasses [&NiagaraSpriteRendererProperties]}]
 userVariables [
  User.ActionLinesLifetime {type=NiagaraFloat, defaultValue=0, description=""}
  User.ActionLinesSpawnRate {type=NiagaraFloat, defaultValue=0, description=""}
  User.bIsFrontEnd {type=NiagaraBool}
  User.bIsFrontEndPreview {type=NiagaraBool}
  User.IsFrontEnd {type=NiagaraBool, defaultValue=true}
  User.IsLargeBody {type=NiagaraBool}
 ]
}
