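# Niagara system for the small crack effect of the MidBreakStretch emote, a brief burst of cartoon-style impact marks.
# ActionLines bursts a few thin black streaks from a tiny sphere, slowed by heavy drag as their length grows and collapses.
# LightningBolts bursts SubUV sprites that spin either way with decaying speed, flickering in size through stacked scale curves.
# A time-of-day color function tints only the bolts.
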NiagaraSystem /SparksCosmetics/Emotes/Bass/MidBreakStretch/Effects/NS_Emote_MidBreakStretch_SmallCrack
dependencies {
 usedDataInterfaces [
  &NiagaraDataInterfaceCurve
  &NiagaraDataInterfaceVector2DCurve
  &NiagaraDataInterfaceSpriteRendererInfo
  &NiagaraDataInterfaceArrayInt32
 ]
 usedDynamicInputs [&Multiply_Vector2D]
 usedModules [&SystemState]
}
emitters {
 ActionLines {
  emitterUpdateScript [
   &EmitterState()
   &SpawnBurst_Instantaneous(
    Spawn Count=&RandomRangeInt(Minimum=4, Maximum=6)
    Use Spawn Probability=false
    UseLoopCountLimit=false
   ) {module=SpawnBurst_Instantaneous001}
  ]
  particleSpawnScript [
   &InitializeParticle(
    Color=<rgba 0 0 0 1>
    Lifetime Max=0.3
    Lifetime Min=0.15
    Lifetime Mode=::NewEnumerator1
    Sprite Size Max=<1 18>
    Sprite Size Min=<0.5 12>
    Sprite Size Mode=::NewEnumerator2
    UsePositionOffset=false
    Write Material Random=true
   )
   &ShapeLocation(
    MaskSpawnGroup=false
    Sphere Distribution=::NewEnumerator1
    Sphere Radius=0.1
    U Position=&Add_Float(
     A=&ReturnNormalizedExecIndex()
     B=&RandomRangeFloat(Minimum=-0.05, Maximum=0.05)
    )
    UseOwnerScale=true
    V Position=&ReturnNormalizedExecIndex()
   )
   &AddVelocity(Module.Use Falloff Distance=false, Override Seed=false, Velocity Mode=::NewEnumerator1, Velocity Speed=400)
   &CameraOffset(Camera Offset Amount=-10)
  ]
  particleUpdateScript [
   &Drag(Drag=&RandomRangeFloat(Minimum=18, Maximum=24), Use Linear Drag=true, Use Rotational Drag=false)
   &SolveForcesAndVelocity(Clamp Velocity=false, Limit Acceleration=false)
   &ParticleState()
   &ScaleSpriteSize(
    Scale Factor=&MakeVector2D(
     X=&FloatFromCurve(FloatCurve=DataInterface() {propertyValues {Curve {keys [<at 0 1>,<at 1 1>]}}})
     Y=&FloatFromCurve(FloatCurve=DataInterface() {propertyValues {Curve {keys [<at 0 0>,<at 0.15 1>,<at 0.4 1>,<at 1 0>]}}})
    )
    Scale Sprite Size Mode=::NewEnumerator2
   )
  ]
  rendererClasses [&NiagaraSpriteRendererProperties]
  renderers [{rendererClass=&NiagaraSpriteRendererProperties}]
 }
 LightningBolts {
  emitterUpdateScript [
   &EmitterState()
   &SpawnBurst_Instantaneous(Spawn Count=4, Use Spawn Probability=false, UseLoopCountLimit=false)
  ]
  particleSpawnScript [
   &InitializeParticle(
    Lifetime=&RandomRangeFloat(Minimum=0.3, Maximum=0.35)
    Sprite Size Max=<35 65>
    Sprite Size Min=<25 55>
    Sprite Size Mode=::NewEnumerator2
    UsePositionOffset=false
    Write Material Random=true
   )
   &PartitionParticles(Number Of Partitions=3)
   &SubUVAnimation(
    Direct Frame Index=1
    "Sprite Renderer "=DataInterface() {propertyValues {SpriteRenderer=&self:LightningBolts_0.NiagaraSpriteRendererProperties_0}}
    SubUV Animation Mode:ENiagara_SubUVLookupModeV2=::NewEnumerator5
   )
   &ShapeLocation(
    MaskSpawnGroup=false
    Offset Coordinate Space=::Local
    Offset Mode=::NewEnumerator0
    Rotation Coordinate Space=::Simulation
    Sphere Distribution=::NewEnumerator1
    Sphere Radius=8
    U Position=&ReturnNormalizedExecIndex(Normalized Index Includes One=false)
    UseOwnerScale=true
    V Position=&ReturnNormalizedExecIndex()
   )
   &AddVelocity(
    Module.Use Falloff Distance=false
    Override Seed=false
    Velocity Mode=::NewEnumerator1
    Velocity Speed=&RandomRangeFloat(Minimum=75, Maximum=150)
   )
   &CameraOffset(Camera Offset Amount=-25)
  ]
  particleUpdateScript [
   &ParticleState()
   /SetVariables(Particles.SpriteScale:NiagaraFloat=&FloatFromCurve(FloatCurve=DataInterface() {propertyValues {Curve {keys [
         {interpMode=RCIM_Cubic, tangentMode=RCTM_User, time=0, value=0}
         {interpMode=RCIM_Cubic, tangentMode=RCTM_User, time=0.1, value=1}
         {arriveTangent=-2.6589, tangentMode=RCTM_User, time=1, value=0}
        ]}}})) {isSetParameters=true}
   /SetVariables(Particles.SpriteSize=&Multiply_Vector2D_ByFloat(Float=$Particles.SpriteScale, Vector2D=$Particles.Initial.SpriteSize)) {isSetParameters=true}
   &ScaleSpriteSize(Uniform Curve Sprite Scale=DataInterface() {propertyValues {Curve {keys [
        {interpMode=RCIM_Cubic, tangentMode=RCTM_User, time=0, value=0}
        {interpMode=RCIM_Cubic, tangentMode=RCTM_User, time=0.05, value=1}
        {arriveTangent=-2.6589, tangentMode=RCTM_User, time=1, value=0}
       ]}}})
   &ScaleSpriteSize(
    Scale Sprite Size Mode=::NewEnumerator0
    Uniform Scale Factor=&FloatFromCurve(FloatCurve=DataInterface() {propertyValues {Curve {keys [<at 0 1>,<at 0.1 0.55>,<at 0.3 1>,<at 0.5 0.55>,<at 0.7 1>,<at 0.9 0.55>,<at 1 1>]}}})
   ) {module=ScaleSpriteSize001}
   &ScaleSpriteSize(
    Non-Uniform Curve Sprite Scale=DataInterface() {propertyValues {
      XCurve {keys [{tangentMode=RCTM_User, time=0, value=0}, {tangentMode=RCTM_User, time=1, value=1}]}
      YCurve {keys [{arriveTangent=5.7604, interpMode=RCIM_Cubic, leaveTangent=5.7604, tangentMode=RCTM_User, time=0, value=0}, <at 0.5 1 RCIM_Cubic>]}
     }}
    Scale Sprite Size Mode=::NewEnumerator3
   ) {module=ScaleSpriteSize002}
   &SpriteRotationRate(Rotation Rate=&Multiply_FloatByInt(
     Float=&FloatFromCurve(
      FloatCurve=DataInterface() {propertyValues {Curve {keys [<at 0 1 RCIM_Cubic>, {arriveTangent=-2.1844, interpMode=RCIM_Cubic, leaveTangent=-2.1844, tangentMode=RCTM_User, time=1, value=0}]}}}
      Scale Curve=&Multiply_Float(A=5000, B=&RandomRangeFloat(Minimum=0.6, Maximum=1.2))
     )
     Integer=&SelectIntFromArray(
      Array Sampling Mode=::NewEnumerator1
      Direct Array Index=?
      Int Selection Array=DataInterface() {propertyValues {IntData [1,-1]}}
      Random Seed=0
     )
    ))
   &Drag(Drag=10, Use Linear Drag=true, Use Rotational Drag=true) {module=Drag001}
   &SolveForcesAndVelocity(Clamp Velocity=false, Limit Acceleration=false)
   &DynamicMaterialParameters(Write Parameter Index 0=::NewEnumerator0)
   &ScaleColor(
    Scale RGB=&MakeVectorFromLinearColorRGB(Color=&FNTimeOfDayColor(
      Night Color=<rgba 0.7152 0.6979 1 1>
      Morning Color=<rgba 0.9922 0.7751 0.8928 1>
      Day Color=<rgba 1 1 1 1>
      Evening Color=<rgba 0.9922 0.5725 0.8 1>
     ))
    ScaleA=true
    ScaleRGB=true
   )
  ]
  rendererClasses [&NiagaraSpriteRendererProperties]
  renderers [{rendererClass=&NiagaraSpriteRendererProperties}]
 }
}
summary {
 systemName=NS_Emote_MidBreakStretch_SmallCrack
 emitters [
  {emitterName=ActionLines, rendererClasses [&NiagaraSpriteRendererProperties]}
  {emitterName=LightningBolts, rendererClasses [&NiagaraSpriteRendererProperties]}
 ]
 userVariables [
  User.bIsFrontEnd {type=NiagaraBool}
  User.bIsFrontEndPreview {type=NiagaraBool}
  User.IsFrontEnd {type=NiagaraBool, defaultValue=true}
  User.IsLargeBody {type=NiagaraBool}
  User.Skeletal Mesh {
   type=NiagaraDataInterfaceSkeletalMesh
   defaultValue {
    struct=NiagaraExt_VariableValue_DataInterface
    value {dataInterface=&self:NiagaraDataInterfaceSkeletalMesh_0}
   }
  }
 ]
}
